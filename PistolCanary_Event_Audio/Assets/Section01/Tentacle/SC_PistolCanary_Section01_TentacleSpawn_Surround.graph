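# Sound cue for the surround tentacle spawn sound in Section 01 of the PistolCanary event.
# A single wave player feeds a 5.1 spawn recording straight to the output, slightly boosted, with tentacle surround attenuation and the Fort_Events sound class.
# A concurrency override allows only one instance and blocks new plays while it is active.
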SoundCue /PistolCanary_Event_Audio/Assets/Section01/Tentacle/SC_PistolCanary_Section01_TentacleSpawn_Surround
root=WavePlayer_0
nodes [{id=WavePlayer_0, properties {soundWaveAssetPtr=/PistolCanary_Event_Audio/Assets/Section01/wav/PistolCanary_Sec_01_Tentacle_Spawn_5_1}}]
properties {
 attenuationSettings=/PistolCanary_Audio/Assets/Settings/ATT_PistolCanary_Tentacle_Surround
 bOverrideConcurrency=true
 maxDistance=40000
 soundClassObject=&Fort_Events
 volumeMultiplier=1.1
 concurrencyOverrides {maxCount=1, resolutionRule=PreventNew}
}
