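# Sound cue for the tentacle eye taunt in Section 01 of the PistolCanary event audio.
# It plays a single wave through one wave player node, with no randomization or modulation.
# Playback uses the tentacle-tip attenuation preset and the Fort_Events sound class, with a slight volume boost.
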SoundCue /PistolCanary_Event_Audio/Assets/Section01/Tentacle/SC_PistolCanary_Section01_Tentacle_EyeTaunt
root=WavePlayer_0
nodes [{id=WavePlayer_0, properties {soundWaveAssetPtr=/PistolCanary_Event_Audio/Assets/Section01/wav/PistolCanary_Sec_01_Tentacle_Eye_Taunt_01}}]
properties {
 attenuationSettings=/PistolCanary_Audio/Assets/Settings/ATT_PistolCanary_Tentacle_Tip
 maxDistance=100000
 soundClassObject=&Fort_Events
 volumeMultiplier=1.12
}
