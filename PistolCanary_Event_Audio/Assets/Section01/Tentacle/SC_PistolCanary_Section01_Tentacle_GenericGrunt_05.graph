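# Sound cue for the fifth generic grunt variant of the tentacle in Section 01 of the PistolCanary event.
# A single wave player feeds the Sec_02 grunt wave straight to the output, with no randomization or modulation.
# It uses the Sec_03 tentacle-tip attenuation with a very large max distance, plays in the Fort_Events sound class and has reduced volume.
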SoundCue /PistolCanary_Event_Audio/Assets/Section01/Tentacle/SC_PistolCanary_Section01_Tentacle_GenericGrunt_05
root=WavePlayer_0
nodes [{id=WavePlayer_0, properties {soundWaveAssetPtr=/PistolCanary_Event_Audio/Assets/Section01/wav/PistolCanary_Sec_02_Generic_Grunt_05}}]
properties {
 attenuationSettings=/PistolCanary_Audio/Assets/Settings/ATT_PistolCanary_Tentacle_Tip_Sec_03
 maxDistance=108000
 soundClassObject=&Fort_Events
 volumeMultiplier=0.6626
}
