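# Sound cue for the tentacle head whip on spawn in Section 01 of the PistolCanary event.
# A single wave player feeds the head-whip wave straight to the output, slightly boosted in volume and routed to the Fort_Events sound class.
# Tentacle-tip attenuation spatializes it, and overridden concurrency caps simultaneous instances per owner with a short retrigger time.
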SoundCue /PistolCanary_Event_Audio/Assets/Section01/Tentacle/SC_PistolCanary_Section01_Tentacle_HeadWhip
root=WavePlayer_0
nodes [{id=WavePlayer_0, properties {soundWaveAssetPtr=/PistolCanary_Event_Audio/Assets/Section01/wav/PistolCanary_Sec01_Tentacle_Spawn_HeadWhip}}]
properties {
 attenuationSettings=/PistolCanary_Audio/Assets/Settings/ATT_PistolCanary_Tentacle_Tip
 bOverrideConcurrency=true
 maxDistance=100000
 soundClassObject=&Fort_Events
 volumeMultiplier=1.2
 concurrencyOverrides {bLimitToOwner=true, maxCount=2, retriggerTime=0.25}
}
