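# Sound cue for the tentacle spawn roar in Section 01 of the PistolCanary event.
# A single wave player plays the roar wave at doubled volume through a dedicated tentacle roar attenuation, in the Fort_Events sound class.
# Bus sends feed sidechain buses named for weapon volume and music high-pass filtering, with the stronger send to the weapon bus, apparently to duck those sounds under the roar.
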SoundCue /PistolCanary_Event_Audio/Assets/Section01/Tentacle/SC_PistolCanary_Section01_Tentacle_Roar
root=WavePlayer_1
nodes [{id=WavePlayer_1, properties {soundWaveAssetPtr=/PistolCanary_Event_Audio/Assets/Section01/wav/PistolCanary_Sec01_Tentacle_Spawn_Roar}}]
properties {
 attenuationSettings=/PistolCanary_Audio/Assets/Settings/ATT_PistolCanary_Tentacle_Roar
 maxDistance=2097152
 soundClassObject=&Fort_Events
 volumeMultiplier=2
 busSends [
  {audioBus=&AB_Sidechain_FortWeapon_Vol, minSendLevel=0, sendLevel=0.7, soundSourceBus=null}
  {audioBus=/Game/Sounds/Modulation/EnvelopeFollowers/HPF/AB_SideChain_FortMusic_HPF, minSendLevel=0, sendLevel=0.2, soundSourceBus=null}
 ]
}
