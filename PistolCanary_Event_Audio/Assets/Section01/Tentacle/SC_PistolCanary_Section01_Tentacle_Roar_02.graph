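# Sound cue for the second Section01 tentacle roar in the PistolCanary event audio.
# A mixer layers two delayed waves reused from Section02 hero intro audio: a roar after a short fixed delay, and a quieter camera-whoosh tail several seconds later.
# It uses a tentacle-roar attenuation asset and the Fort_Events sound class, with a boosted overall volume.
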SoundCue /PistolCanary_Event_Audio/Assets/Section01/Tentacle/SC_PistolCanary_Section01_Tentacle_Roar_02
root=Mixer_0
nodes [
 {id=Delay_0, children [WavePlayer_0], properties {delayMax=1, delayMin=1}}
 {id=Delay_1, children [WavePlayer_1], properties {delayMax=5.5, delayMin=5.5}}
 {id=Mixer_0, children [Delay_0,Delay_1], properties {inputVolume [1,0.35]}}
 {
  id=WavePlayer_0
  properties {soundWaveAssetPtr=/PistolCanary_Event_Audio/Assets/Section02/Heroes/wav/PistolCanary_Character_Intro_01_Rawr_Post}
 }
 {
  id=WavePlayer_1
  properties {soundWaveAssetPtr=/PistolCanary_Event_Audio/Assets/Section02/Heroes/wav/PistolCanary_Character_Intro_01_CameraWhoosh_End}
 }
]
properties {
 attenuationSettings=/PistolCanary_Audio/Assets/Settings/ATT_PistolCanary_Tentacle_Roar
 maxDistance=2097152
 soundClassObject=&Fort_Events
 volumeMultiplier=1.6
}
