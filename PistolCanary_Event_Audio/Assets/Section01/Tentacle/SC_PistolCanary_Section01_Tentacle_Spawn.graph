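# Sound cue for the tentacle spawn sound in Section 01 of the PistolCanary event.
# It plays a single wave through one wave player on the Fort_Events sound class, with a dedicated spawn attenuation asset, a slight volume boost and a very large max distance.
# Bus sends feed sidechain audio buses for music low-pass and high-pass filtering and weapon volume.
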SoundCue /PistolCanary_Event_Audio/Assets/Section01/Tentacle/SC_PistolCanary_Section01_Tentacle_Spawn
root=WavePlayer_0
nodes [{id=WavePlayer_0, properties {soundWaveAssetPtr=/PistolCanary_Event_Audio/Assets/Section01/wav/PistolCanary_Sec01_Tentacle_Spawn}}]
properties {
 attenuationSettings=/PistolCanary_Event_Audio/Assets/Settings/ATT_PistolCanary_Tentacle_Spawn
 maxDistance=103000
 soundClassObject=&Fort_Events
 volumeMultiplier=1.2
 busSends [
  {audioBus=/Game/Sounds/Modulation/EnvelopeFollowers/LPF/AB_SideChain_FortMusic_LPF, minSendLevel=0, sendLevel=0.25, soundSourceBus=null}
  {audioBus=&AB_Sidechain_FortWeapon_Vol, minSendLevel=0, sendLevel=0.5, soundSourceBus=null}
  {audioBus=/Game/Sounds/Modulation/EnvelopeFollowers/HPF/AB_SideChain_FortMusic_HPF, minSendLevel=0, sendLevel=0.5, soundSourceBus=null}
 ]
}
