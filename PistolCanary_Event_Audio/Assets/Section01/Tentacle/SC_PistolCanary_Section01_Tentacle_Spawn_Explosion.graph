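# Sound cue for the tentacle spawn explosion in Section 01 of the PistolCanary event.
# A single wave player feeds the root, playing the portal explode tentacle reveal wave.
# It uses the event's explosion attenuation asset and the Fort_Events sound class, with slightly reduced volume and a very large max distance.
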SoundCue /PistolCanary_Event_Audio/Assets/Section01/Tentacle/SC_PistolCanary_Section01_Tentacle_Spawn_Explosion
root=WavePlayer_1
nodes [{
  id=WavePlayer_1
  properties {soundWaveAssetPtr=/PistolCanary_Event_Audio/Assets/Section01/wav/PistolCanary_Sec_01_Portal_Explode_Tentacle_Reveal}
 }]
properties {
 attenuationSettings=/PistolCanary_Audio/Assets/Settings/ATT_PistolCanary_Explosions
 maxDistance=130000
 soundClassObject=&Fort_Events
 volumeMultiplier=0.8
}
